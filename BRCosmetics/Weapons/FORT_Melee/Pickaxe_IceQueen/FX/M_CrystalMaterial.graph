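# Unlit additive material for mesh particles in the Ice Queen pickaxe FX, giving a view-dependent, hue-shifting crystal look.
# The sine of the normal map's absolute dot with the view vector, plus scaled UV2, drives HueShift on a base color.
# That color is tinted by particle color, masked by inverted emissive-texture blue, then passed through emissive adjustments.
# The squared view term times particle alpha and vertex red forms a fade; CameraOffset supplies a small offset.
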Material /BRCosmetics/Weapons/FORT_Melee/Pickaxe_IceQueen/FX/M_CrystalMaterial
graph {
 properties {bUsedWithMeshParticles=true, blendMode=BLEND_Additive, pixelDepthOffsetMode=PDOM_Legacy, shadingModel=MSM_Unlit, usedShadingModels=Unlit}
 nodes [
  MaterialFunctionCall(Offset Amount=Constant() {r=0.1}) {materialFunction=/Engine/Functions/Engine_MaterialFunctions02/WorldPositionOffset/CameraOffset}
  n1 = Transform(None=TextureSample() {samplerType=SAMPLERTYPE_Normal, texture=/BRCosmetics/Weapons/FORT_Melee/Pickaxe_IceQueen/Textures/T_Pickaxe_IceQueen_N}.RGB)
  n2 = Normalize(VectorInput=Subtract(A=WorldPosition().XYZ, B=CameraPositionWS()))
  n3 = Sine(None=Abs(None=DotProduct(A=@n1, B=@n2)))
  n4 = TextureCoordinate() {coordinateIndex=2}
  n5 = MaterialFunctionCall(Hue Shift Percentage=Add(A=Abs(None=@n3), B=Multiply(A=@n4) {constB=0.35}), Texture=Constant3Vector() {constant=<rgba 0.0908 0.15 0.0405 0>}) {materialFunction=&HueShift}
  n6 = ParticleColor()
  n7 = TextureSample() {texture=/BRCosmetics/Weapons/FORT_Melee/Pickaxe_IceQueen/Textures/T_Pickaxe_IceQueen_E}
  MaterialFunctionCall(Emissive=Multiply(A=Multiply(A=@n5.Result, B=@n6.RGB), B=OneMinus(None=@n7.B))) {materialFunction=&MF_FX_EmissiveAdjustments}
  Multiply(A=Multiply(A=@n3, B=@n3), B=Multiply(A=@n6.A, B=VertexColor().R))
  VertexNormalWS()
 ]
}
