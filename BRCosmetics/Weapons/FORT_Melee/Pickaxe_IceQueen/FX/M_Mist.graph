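# Translucent unlit particle material for mist on the Ice Queen pickaxe FX.
# Two noise-distorted panning streak textures, timed and randomized by dynamic parameters, are multiplied and masked by a squared sine edge fade across U.
# On deferred, a scene-normal Fresnel near depth intersections is added; forward and mobile use a constant.
# Scaled by particle alpha, the result indexes a blue gradient texture for tinted emissive and masked alpha.
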Material /BRCosmetics/Weapons/FORT_Melee/Pickaxe_IceQueen/FX/M_Mist
graph {
 properties {
  bUsedWithParticleSprites=true
  blendMode=BLEND_Translucent
  pixelDepthOffsetMode=PDOM_Legacy
  shadingModel=MSM_Unlit
  usedShadingModels=Unlit
 }
 nodes [
  n1 = ParticleColor()
  n2 = CameraVectorWS()
  n3 = ComponentMask(None=SceneTexture() {sceneTextureId=PPI_WorldNormal}.Color) {b=true, g=true, r=true}
  n4 = ScalarParameter() {defaultValue=50, parameterName=DF_FD}
  n5 = Multiply(A=OneMinus(None=DotProduct(A=@n2, B=@n3)), B=OneMinus(None=DepthFade(FadeDistance=@n4)))
  n6 = Constant()
  n7 = VectorParameter() {parameterName=UV, defaultValue=<rgba 0.5 1 0.2 2>}
  n8 = DynamicParameter() {paramNames [Time,Random,Param3,Param4]}
  n9 = Panner(
   Coordinate=Multiply(A=TextureCoordinate(), B=ComponentMask(None=@n7.RGB) {g=true, r=true})
   Time=Add(A=@n8.Time, B=@n8.Time)
   Speed=AppendVector(A=@n7.B, B=@n7.B)
  ) {speedY=0.2}
  n10 = TextureSample(UVs=Add(A=TextureCoordinate() {uTiling=0.2, vTiling=0.2}, B=@n8.Random)) {samplerType=SAMPLERTYPE_LinearColor, texture=&T_NoisePack_LinearColor}
  n11 = Multiply(A=Subtract(A=ComponentMask(None=@n10.RGB) {g=true, r=true}) {constB=0.5}, B=ScalarParameter() {defaultValue=0.3, parameterName=Offset}) {constB=0.1}
  n12 = TextureSample(UVs=MaterialFunctionCall(XY=Add(A=@n9, B=@n11)) {materialFunction=&Swizzle}.YX) {samplerSource=SSM_Wrap_WorldGroupSettings, texture=/Game/Effects/Fort_Effects/Textures/Energy/T_OutpostShockwaveStreaks_02}
  n13 = Constant2Vector() {g=0.618, r=1.256}
  n14 = TextureSample(UVs=Add(A=Multiply(A=@n9, B=@n13), B=@n11)) {samplerSource=SSM_Wrap_WorldGroupSettings, texture=/Game/Global/Textures/Default/Masks_and_Gradients/T_BlasterBeamDistort}
  n15 = SquareRoot(None=Multiply(A=Multiply(A=@n12.R, B=@n14.R)) {constB=2})
  n16 = MaterialFunctionCall(Float2=Sine(None=TextureCoordinate() {uTiling=0.5, vTiling=0.5})) {materialFunction=&BreakOutFloat2Components}
  n17 = Multiply(A=@n16.R, B=@n16.R)
  n18 = Multiply(A=@n1.A, B=Add(A=ShadingPathSwitch(Default=@n5, Deferred=@n5, Forward=@n6, Mobile=@n6), B=Multiply(A=@n15, B=@n17)))
  CurveAtlasRowParameter(CurveTime=@n18) {atlas=&AthenaCurveAtlas01, curve=/Game/Athena/LinearColorCurves/Curve_IcoQueenColor, defaultValue=0.1621}
  n19 = TextureSampleParameter2D(UVs=@n18) {parameterName=CurveTexture, texture=&/T_IceQueenBlueGradient}
  MaterialFunctionCall(Emissive=Multiply(A=@n19.RGB, B=@n1.RGB)) {materialFunction=&MF_FX_EmissiveAdjustments}
  Multiply(A=@n19.A, B=@n17)
  Multiply(A=@n9, B=Constant2Vector() {g=0.35, r=0.56})
  TextureSample()
 ]
}
