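# Physics asset for the skeletal mesh of the Ice Queen pickaxe demo, a Fortnite two-handed melee cosmetic.
# A single capsule body on the Melee_TwoHanded_Handle bone approximates the weapon's collision.
# It has no constraints, so it simulates as one rigid piece with no articulated parts.
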PhysicsAsset /BRCosmetics/Weapons/FORT_Melee/Pickaxe_IceQueen/Meshes/SK_Pickaxe_IceQueen_DEMO_PhysicsAsset
bodies {Melee_TwoHanded_Handle [{shapeType=Capsule, length=83.0667, radius=40.4035, center=<0.3242 2.3195 36.8564>, rotation=<rot 0.5913 -90 -0.0052>}]}
constraints []
